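# Niagara effect for the balloon pop in the RevoltCrush emote: flung scraps, air streaks and a brief time-of-day-tinted flash.
# BalloonBitsCondensed merges several layers into one emitter, using spawn groups to switch per-layer values and a visibility tag across sprite and mesh renderers.
# Owner velocity adjusts streak speed, scrap velocity inheritance and world/local position blending, while front-end modifiers alter scrap velocity and camera offset.
# DarkBurst and SmokeLWE carry no renderers.
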NiagaraSystem /BRCosmetics/Animation/Game/MainPlayer/Emotes/RevoltCrush/FX/NS_Emote_RevoltCrush_BalloonPop
dependencies {
 usedDataInterfaces [
  &NiagaraDataInterfaceVector2DCurve
  &NiagaraDataInterfaceCurve
  &NiagaraDataInterfaceArrayFloat2
  &NiagaraDataInterfaceSpriteRendererInfo
  &NiagaraDataInterfaceVectorCurve
 ]
 usedDynamicInputs []
 usedModules [&SystemState, &BlendSpaceHelper]
}
emitters {
 AirStreaks {
  emitterUpdateScript [
   &EmitterState()
   &SpawnBurst_Instantaneous(Spawn Count=12, Use Spawn Probability=false, UseLoopCountLimit=false)
  ]
  particleSpawnScript [
   /SetVariables(Particles.RandomFloat=&UniformRangedFloat(Minimum=0, Maximum=3)) {enabled=false, isSetParameters=true}
   &InitializeParticle(
    Color Maximum=<rgba 0.0573 0.0573 0.0573 1>
    Color Minimum=<rgba 0.0135 0.0143 0.0156 3>
    Color Mode=::NewEnumerator2
    Lifetime Max=0.35
    Lifetime Min=0.2
    Lifetime Mode=::NewEnumerator1
    Sprite Size=&Multiply_Vector2D_ByFloat(Float=0.2, Vector2D=&RandomRangeVector2D(Minimum=<50 200>, Maximum=<60 300>))
    Sprite Size Mode=::NewEnumerator4
    UsePositionOffset=false
    Write Material Random=true
   )
   &ShapeLocation(MaskSpawnGroup=false, Override Seed=false, Sphere Radius=25, UseOwnerScale=true)
   &AddVelocity(
    Module.Use Falloff Distance=false
    Override Seed=false
    Velocity Mode=::NewEnumerator1
    Velocity Speed=&MakeCustomFloatFromBool(
     Bool=&SetBoolByFloatComparison(A=&VectorLength(Vector=$User.OwnerLocalVelocity), B=301)
     True Float=0
     False Float=25
    )
   )
   &ApplyInitialForces()
   &SubUVAnimation(End Frame=3, SubUV Animation Mode:ENiagara_SubUVLookupMode=::NewEnumerator2)
  ]
  particleUpdateScript [
   &ParticleState()
   &SolveForcesAndVelocity(Clamp Velocity=false, Limit Acceleration=false)
   &ScaleSpriteSize(Scale Factor=&Vector2DFromCurve(
     Vector2Curve=DataInterface() {propertyValues {XCurve {keys [<at 0.0036 1.6897>,<at 0.9967 0.0582>]}, YCurve {keys [<at 0 1>,<at 1.002 3.7227>]}}}
     ScaleCurve=&Multiply_Vector2D_ByFloat(Float=0.4, Vector2D=<1 1>)
    ))
   &ScaleColor(
    Scale Alpha=&FloatFromCurve(
     FloatCurve=DataInterface() {propertyValues {Curve {keys [
         {interpMode=RCIM_Cubic, leaveTangent=-1.9294, tangentMode=RCTM_Break, time=0, value=1}
         {interpMode=RCIM_Cubic, tangentMode=RCTM_Break, time=1, value=0}
        ]}}}
     Scale Curve=&RandomRangeFloat(Minimum=0.1, Maximum=0.3)
    )
    ScaleA=true
    ScaleRGB=false
   )
   &DynamicMaterialParameters(
    Index 0 Param 2=1
    Index 0 Param 3=1
    Index 0 Param 4=&FloatFromCurve(FloatCurve=DataInterface() {propertyValues {Curve {keys [<at 0.3 1>,<at 1 0.25>]}}})
    Module.Param0WriteEnabled=true
    Module.Param1WriteEnabled=true
    Module.Param2WriteEnabled=true
    Module.Param3WriteEnabled=true
   )
   &BlendSpace("Position Blend (World <-> Local)":NiagaraFloat=&MakeCustomFloatFromBool(
     Bool=&SetBoolByFloatComparison(A=&VectorLength(Vector=$User.OwnerLocalVelocity), B=301)
     True Float=1
     False Float=&FloatFromCurve(FloatCurve=DataInterface() {propertyValues {Curve {keys [<at 0.25 1>,<at 0.5 0>]}}})
     Translation Contribution:NiagaraFloat=1
     Rotation Contribution:NiagaraFloat=0
    ))
  ]
  rendererClasses [&NiagaraSpriteRendererProperties]
  renderers [{rendererClass=&NiagaraSpriteRendererProperties}]
 }
 BalloonBitsCondensed {
  emitterUpdateScript [
   &EmitterState()
   &SpawnBurst_Instantaneous(
    Spawn Count=&RandomRangeInt(Minimum=6, Maximum=9)
    Use Spawn Probability=false
    UseLoopCountLimit=false
   )
   &SpawnBurst_Instantaneous(
    Spawn Count=&RandomRangeInt(Minimum=8, Maximum=12)
    Spawn Group=1
    Use Spawn Probability=false
    UseLoopCountLimit=false
   ) {module=SpawnBurst_Instantaneous001}
   &SpawnBurst_Instantaneous(
    Spawn Count=&RandomRangeInt(Minimum=3, Maximum=4)
    Spawn Group=2
    Use Spawn Probability=false
    UseLoopCountLimit=false
   ) {module=SpawnBurst_Instantaneous002}
   &SpawnBurst_Instantaneous(Spawn Group=3, Use Spawn Probability=false, UseLoopCountLimit=false) {module=SpawnBurst_Instantaneous003}
   &SpawnBurst_Instantaneous(Spawn Group=4, Use Spawn Probability=false, UseLoopCountLimit=false) {module=SpawnBurst_Instantaneous004}
  ]
  particleSpawnScript [
   /SetVariables(Particles.VisibilityTag=&SpawnGroup()) {isSetParameters=true}
   &InitializeParticle(
    Color=&MergeHelper_ColorSwitch(
     0=&MultiplyLinearColor(
      Linear Color=&RandomRangeLinearColor(Minimum=<rgba 0.1458 0.1369 0.1329 2>, Maximum=<rgba 0.6198 0.5812 0.5505 2>)
      Scale Factor=&MakeFloat4FromVectorAndFloat(
       "Vector (RGB)"=&VectorFromFloat(Value=&Lerp_Float(
         A=0.1
         B=1
         Alpha=&FloatFromCurve(
          FloatCurve=DataInterface() {propertyValues {Curve {keys [{interpMode=RCIM_Cubic, tangentMode=RCTM_User, time=0, value=0}, {interpMode=RCIM_Cubic, tangentMode=RCTM_User, time=1, value=1}]}}}
          CurveIndex=&RandomRangeFloat(Minimum=0, Maximum=1)
         )
        ))
       "Float (Alpha)"=1
      )
     )
     1=&MultiplyLinearColor(
      Linear Color=&RandomRangeLinearColor(Minimum=<rgba 0.1458 0.1369 0.1329 2>, Maximum=<rgba 0.6198 0.5812 0.5505 2>)
      Scale Factor=&MakeFloat4FromVectorAndFloat(
       "Vector (RGB)"=&VectorFromFloat(Value=&Lerp_Float(
         A=0.1
         B=1
         Alpha=&FloatFromCurve(
          FloatCurve=DataInterface() {propertyValues {Curve {keys [{interpMode=RCIM_Cubic, tangentMode=RCTM_User, time=0, value=0}, {interpMode=RCIM_Cubic, tangentMode=RCTM_User, time=1, value=1}]}}}
          CurveIndex=&RandomRangeFloat(Minimum=0, Maximum=1)
         )
        ))
       "Float (Alpha)"=1
      )
     )
     2=&RandomRangeLinearColor(Minimum=<rgba 0.026 0.026 0.026 0.5>, Maximum=<rgba 0.0729 0.0729 0.0729 0.4>)
     3=<rgba 0.0104 0.0104 0.0104 50>
     4=<rgba 0.0104 0.0104 0.0104 5>
     5=&RandomRangeLinearColor(Minimum=<rgba 0.0135 0.0143 0.0156 3>, Maximum=<rgba 0.0573 0.0573 0.0573 1>)
     6=&FN_Select_Color_from_Time_Of_Day(
      Morning=<rgba 0.3125 0.3125 0.3125 1>
      Day=<rgba 0.4601 0.4903 0.5104 1>
      Evening=<rgba 0.3125 0.3125 0.3125 1>
      Night=<rgba 0.1648 0.2191 0.3229 1>
     )
     Index=&SpawnGroup()
     Num=::NewEnumerator3
    )
    Lifetime=&MergeHelper_FloatSwitch(
     0=&Lerp_Float(
      A=0.4
      B=1.2
      Alpha=&FloatFromCurve(
       FloatCurve=DataInterface() {propertyValues {Curve {keys [
           {interpMode=RCIM_Cubic, tangentMode=RCTM_User, time=0, value=0}
           {arriveTangent=2.9148, interpMode=RCIM_Cubic, leaveTangent=2.9148, tangentMode=RCTM_User, time=1, value=1}
          ]}}}
       CurveIndex=&RandomRangeFloat(Minimum=0, Maximum=1)
      )
     )
     1=&Lerp_Float(
      A=0.4
      B=1.2
      Alpha=&FloatFromCurve(
       FloatCurve=DataInterface() {propertyValues {Curve {keys [
           {interpMode=RCIM_Cubic, tangentMode=RCTM_User, time=0, value=0}
           {arriveTangent=2.9148, interpMode=RCIM_Cubic, leaveTangent=2.9148, tangentMode=RCTM_User, time=1, value=1}
          ]}}}
       CurveIndex=&RandomRangeFloat(Minimum=0, Maximum=1)
      )
     )
     2=&RandomRangeFloat(Minimum=0.3, Maximum=1)
     3=&RandomRangeFloat(Minimum=0.15, Maximum=0.2)
     4=0.12
     5=&RandomRangeFloat(Minimum=0.2, Maximum=0.35)
     Index=&SpawnGroup()
     Num=::NewEnumerator3
    )
    Mesh Scale=&MaskVector(
     Mask Source=0
     Mask Parameter=?
     Mask=2
     Masked Vector=&VectorFromFloat(Value=&RandomRangeFloat(Minimum=0.05, Maximum=0.3))
     Passthrough Vector=&MaskVector(
      Mask Source=0
      Mask Parameter=?
      Mask=3
      Masked Vector=&Multiply_VectorByFloat(Vector=<1.5 1.5 1.5>, Float=0.6)
      Passthrough Vector=<0 0 0>
     )
    )
    Mesh Scale Mode=::NewEnumerator4
    Sprite Rotation Angle=&MergeHelper_FloatSwitch(
     0=&RandomRangeFloat(Minimum=-360, Maximum=360)
     1=&RandomRangeFloat(Minimum=-360, Maximum=360)
     2=0
     3=0
     4=0
     Index=&SpawnGroup()
     Num=::NewEnumerator3
    )
    Sprite Rotation Mode=::NewEnumerator0
    Sprite Size=&MergeHelper_Vector2Switch(
     0=&SelectVector2DFromArray(
      Array Sampling Mode=::NewEnumerator0
      Direct Array Index=?
      Vector 2D Selection Array=DataInterface() {propertyValues {FloatData [<5 8>, <8 5>, <2 3>, <3 2>, <11 13>, <13 11>, <4 6>, <6 4>, <2.5 2>, <2 2.5>, <12 15>, <15 12>]}}
      Random Seed=0
     )
     1=&SelectVector2DFromArray(
      Array Sampling Mode=::NewEnumerator0
      Direct Array Index=?
      Vector 2D Selection Array=DataInterface() {propertyValues {FloatData [<5 8>, <8 5>, <2 3>, <3 2>, <11 13>, <13 11>, <4 6>, <6 4>, <2.5 2>, <2 2.5>, <12 15>, <15 12>]}}
      Random Seed=0
     )
     2=<0 0>
     3=<0 0>
     4=<100 100>
     5=&Multiply_Vector2D_ByFloat(Float=0.2, Vector2D=&RandomRangeVector2D(Minimum=<50 200>, Maximum=<60 300>))
     Index=&SpawnGroup()
     Num=::NewEnumerator3
    )
    Sprite Size Mode=::NewEnumerator4
    UsePositionOffset=false
    Write Material Random=true
   )
   &ShapeLocation(
    MaskSpawnGroup=false
    Override Seed=false
    Sphere Radius=&MergeHelper_FloatSwitch(
     0=&RandomRangeFloat(Minimum=8, Maximum=10)
     1=&RandomRangeFloat(Minimum=8, Maximum=10)
     2=5
     3=0
     4=0
     Index=&SpawnGroup()
     Num=::NewEnumerator3
    )
    UseOwnerScale=true
   )
   &SubUVAnimation(
    End Frame Range Override=22
    Output Alternate Renderer Binding=true
    "Sprite Renderer "=DataInterface() {propertyValues {SpriteRenderer=&self:BalloonBitsCondensed_0.NiagaraSpriteRendererProperties_1}}
    Start Frame Range Override=11
    SubUV Animation Mode:ENiagara_SubUVLookupModeV2=::NewEnumerator2
    UseEndFrame=true
    UseStartFrame=true
   )
   &SubUVAnimation(
    End Frame Range Override=23
    Output Alternate Renderer Binding=true
    "Sprite Renderer "=DataInterface() {propertyValues {SpriteRenderer=&self:BalloonBitsCondensed_0.NiagaraSpriteRendererProperties_2}}
    Start Frame Range Override=14
    SubUV Animation Mode:ENiagara_SubUVLookupModeV2=::NewEnumerator2
    UseEndFrame=true
    UseStartFrame=true
   ) {module=SubUVAnimation001}
   &FN_Sprite_Attributes(
    Create Facing Attribute=true
    Facing:Vector3f=$Particles.ShapeLocation.ShapeNormal
    Lifetime:NiagaraFloat=&MergeHelper_FloatSwitch(
     0=&Lerp_Float(
      A=0.4
      B=1.2
      Alpha=&FloatFromCurve(
       FloatCurve=DataInterface() {propertyValues {Curve {keys [
           {interpMode=RCIM_Cubic, tangentMode=RCTM_User, time=0, value=0}
           {arriveTangent=2.9148, interpMode=RCIM_Cubic, leaveTangent=2.9148, tangentMode=RCTM_User, time=1, value=1}
          ]}}}
       CurveIndex=&RandomRangeFloat(Minimum=0, Maximum=1)
      )
     )
     1=&Lerp_Float(
      A=0.4
      B=1.2
      Alpha=&FloatFromCurve(
       FloatCurve=DataInterface() {propertyValues {Curve {keys [
           {interpMode=RCIM_Cubic, tangentMode=RCTM_User, time=0, value=0}
           {arriveTangent=2.9148, interpMode=RCIM_Cubic, leaveTangent=2.9148, tangentMode=RCTM_User, time=1, value=1}
          ]}}}
       CurveIndex=&RandomRangeFloat(Minimum=0, Maximum=1)
      )
     )
     2=&RandomRangeFloat(Minimum=0.3, Maximum=1)
     3=&RandomRangeFloat(Minimum=0.15, Maximum=0.2)
     4=0.12
     5=&RandomRangeFloat(Minimum=0.2, Maximum=0.35)
     Index=&SpawnGroup()
     Num=::NewEnumerator3
    )
    Write to Module Namespace=true
   )
   &AddVelocity(
    Module.Use Falloff Distance=false
    Override Seed=false
    Velocity Mode=::NewEnumerator1
    Velocity Speed=&MergeHelper_FloatSwitch(
     0=&Multiply_Float(
      A=300
      B=&FloatFromCurve(FloatCurve=DataInterface() {propertyValues {Curve {keys [
           {arriveTangent=-3.1817, interpMode=RCIM_Cubic, leaveTangent=-3.1817, tangentMode=RCTM_User, time=0, value=1}
           {interpMode=RCIM_Cubic, tangentMode=RCTM_User, time=1, value=0}
          ]}}})
     )
     1=&Multiply_Float(
      A=300
      B=&FloatFromCurve(FloatCurve=DataInterface() {propertyValues {Curve {keys [
           {arriveTangent=-3.1817, interpMode=RCIM_Cubic, leaveTangent=-3.1817, tangentMode=RCTM_User, time=0, value=1}
           {interpMode=RCIM_Cubic, tangentMode=RCTM_User, time=1, value=0}
          ]}}})
     )
     2=&Multiply_Float(
      A=300
      B=&FloatFromCurve(FloatCurve=DataInterface() {propertyValues {Curve {keys [{arriveTangent=-3.3408, interpMode=RCIM_Cubic, leaveTangent=-3.3408, tangentMode=RCTM_User, time=0, value=1}, <at 1 0>]}}})
     )
     3=0
     4=0
     5=&MakeCustomFloatFromBool(
      Bool=&SetBoolByFloatComparison(A=&VectorLength(Vector=$User.OwnerLocalVelocity), B=301)
      True Float=0
      False Float=25
     )
     Index=&SpawnGroup()
     Num=::NewEnumerator3
    )
   )
   &AddVelocity(Velocity=&MergeHelper_VectorSwitch(
     0=&MakeVector(
      X=0
      Y=&FloatFromCurve(FloatCurve=DataInterface() {propertyValues {Curve {defaultValue=1, keys [{tangentMode=RCTM_User, time=0, value=0}, {tangentMode=RCTM_User, time=1, value=-20}]}}})
      Z=0
     )
     1=&MakeVector(
      X=0
      Y=&FloatFromCurve(FloatCurve=DataInterface() {propertyValues {Curve {defaultValue=1, keys [{tangentMode=RCTM_User, time=0, value=0}, {tangentMode=RCTM_User, time=1, value=-20}]}}})
      Z=0
     )
     2=&MakeVector(
      X=0
      Y=&FloatFromCurve(FloatCurve=DataInterface() {propertyValues {Curve {defaultValue=1, keys [{tangentMode=RCTM_User, time=0, value=0}, {tangentMode=RCTM_User, time=1, value=-40}]}}})
      Z=-20
     )
     3=<0 0 0>
     4=<0 0 0>
     Index=&SpawnGroup()
     Num=::NewEnumerator3
    )) {module=AddVelocity001}
   &CameraOffset(Camera Offset Amount=&MergeHelper_FloatSwitch(
     0=&RandomRangeFloat(Minimum=5, Maximum=15)
     1=&RandomRangeFloat(Minimum=5, Maximum=15)
     2=&RandomRangeFloat(Minimum=9, Maximum=15)
     3=&FN_FrontEndFloatModifier(Gameplay=15, Front End=10)
     4=&FN_FrontEndFloatModifier(Gameplay=15, Front End=6)
     Index=&SpawnGroup()
     Num=::NewEnumerator3
    ))
   &ApplyInitialForces()
   &AddVelocity(Velocity=&FN_FrontEndVector3Modifier(
     Gameplay=<0 0 0>
     Front End=&MakeVector(
      X=0
      Y=&MergeHelper_FloatSwitch(0=-425, 1=-425, 2=-425, 3=0, 4=0, Index=&SpawnGroup(), Num=::NewEnumerator3)
      Z=0
     )
    )) {module=AddVelocity002}
  ]
  particleUpdateScript [
   &ParticleState()
   &ScaleMeshSize(Scale Factor=&MaskVector(
     Mask Source=0
     Mask Parameter=?
     Mask=2
     Masked Vector=&Multiply_VectorByFloat(
      Vector=<1 1 1>
      Float=&FloatFromCurve(FloatCurve=DataInterface() {propertyValues {Curve {keys [
           <at 0.0241 1 RCIM_Cubic>
           {arriveTangent=-1.7795, interpMode=RCIM_Cubic, leaveTangent=-1.7795, tangentMode=RCTM_User, time=0.2411, value=0.5209}
           {arriveTangent=-0.0723, interpMode=RCIM_Cubic, leaveTangent=-0.0723, tangentMode=RCTM_User, time=1, value=0}
          ]}}})
     )
     Passthrough Vector=&MaskVector(
      Mask Source=0
      Mask Parameter=?
      Mask=3
      Masked Vector=&VectorFromCurve(VectorCurve=DataInterface() {propertyValues {XCurve {keys [<at 0 0>,<at 0.25 1>]}, YCurve {keys [<at 0 0>,<at 0.25 1>]}, ZCurve {keys [<at 0 0>,<at 0.25 1>]}}})
      Passthrough Vector=<0 0 0>
     )
    ))
   &ScaleSpriteSize(
    Scale Factor=&MergeHelper_Vector2Switch(
     0=&Vector2DFromFloat(Value=&FloatFromCurve(FloatCurve=DataInterface() {propertyValues {Curve {
          defaultValue=0
          keys [
           <at 0 1 RCIM_Cubic>
           {arriveTangent=-1.3254, interpMode=RCIM_Cubic, leaveTangent=-1.3254, time=0.1732, value=0.3874}
           {arriveTangent=-1.8732, interpMode=RCIM_Cubic, leaveTangent=-1.8732, tangentMode=RCTM_User, time=0.3098, value=0.5894}
           <at 1 0 RCIM_Cubic>
          ]
         }}}))
     1=&Vector2DFromFloat(Value=&FloatFromCurve(FloatCurve=DataInterface() {propertyValues {Curve {
          defaultValue=0
          keys [
           <at 0 1 RCIM_Cubic>
           {arriveTangent=-1.3254, interpMode=RCIM_Cubic, leaveTangent=-1.3254, time=0.1732, value=0.3874}
           {arriveTangent=-1.8732, interpMode=RCIM_Cubic, leaveTangent=-1.8732, tangentMode=RCTM_User, time=0.3098, value=0.5894}
           <at 1 0 RCIM_Cubic>
          ]
         }}}))
     2=<0 0>
     3=<0 0>
     4=&Vector2DFromFloat(Value=&FloatFromCurve(FloatCurve=DataInterface() {propertyValues {Curve {keys [{tangentMode=RCTM_User, time=0, value=0.4}, {tangentMode=RCTM_User, time=1, value=0.8}]}}}))
     5=&Vector2DFromCurve(
      Vector2Curve=DataInterface() {propertyValues {XCurve {keys [<at 0.0036 1.6897>,<at 0.9967 0.0582>]}, YCurve {keys [<at 0 1>,<at 1.002 3.7227>]}}}
      ScaleCurve=&Multiply_Vector2D_ByFloat(Float=0.4, Vector2D=<1 1>)
     )
     Index=&SpawnGroup()
     Num=::NewEnumerator3
    )
    Scale Sprite Size Mode=::NewEnumerator2
   ) {module=ScaleSpriteSize001}
   &SpriteRotationRate(Rotation Rate=&MaskFloat(
     Mask Source=0
     Mask Parameter=?
     Mask=0
     Masked Float=&Multiply_Float(
      A=&Lerp_Float(
       A=25
       B=1000
       Alpha=&FloatFromCurve(FloatCurve=DataInterface() {propertyValues {Curve {keys [
            {arriveTangent=-3.2581, interpMode=RCIM_Cubic, leaveTangent=-3.2581, tangentMode=RCTM_User, time=0.1, value=1}
            {interpMode=RCIM_Cubic, tangentMode=RCTM_User, time=1, value=0}
           ]}}})
      )
      B=&FloatFromCurve(
       FloatCurve=DataInterface() {propertyValues {Curve {keys [<at 0 -1 RCIM_Constant>,<at 0.5 1 RCIM_Constant>]}}}
       CurveIndex=&RandomRangeFloat(Minimum=-1, Maximum=1)
      )
     )
     Passthrough Float=&MaskFloat(
      Mask Source=0
      Mask Parameter=?
      Mask=1
      Masked Float=&Multiply_Float(
       A=&Lerp_Float(
        A=25
        B=1000
        Alpha=&FloatFromCurve(FloatCurve=DataInterface() {propertyValues {Curve {keys [
             {arriveTangent=-3.2581, interpMode=RCIM_Cubic, leaveTangent=-3.2581, tangentMode=RCTM_User, time=0.1, value=1}
             {interpMode=RCIM_Cubic, tangentMode=RCTM_User, time=1, value=0}
            ]}}})
       )
       B=&FloatFromCurve(
        FloatCurve=DataInterface() {propertyValues {Curve {keys [<at 0 -1 RCIM_Constant>,<at 0.5 1 RCIM_Constant>]}}}
        CurveIndex=&RandomRangeFloat(Minimum=-1, Maximum=1)
       )
      )
      Passthrough Float=0
     )
    ))
   &AccelerationForce(Acceleration=&MaskVector(
     Mask Source=0
     Mask Parameter=?
     Mask=0
     Masked Vector=&Multiply_VectorByFloat(
      Vector=&RandomRangeVector(Minimum=<300 15 200>, Maximum=<-300 -500 -1000>)
      Float=&FloatFromCurve(FloatCurve=DataInterface() {propertyValues {Curve {
          defaultValue=1
          keys [
           {arriveTangent=-3.5798, interpMode=RCIM_Cubic, leaveTangent=-3.5798, tangentMode=RCTM_User, time=0, value=1}
           {interpMode=RCIM_Cubic, tangentMode=RCTM_User, time=1, value=0}
          ]
         }}})
     )
     Passthrough Vector=&MaskVector(
      Mask Source=0
      Mask Parameter=?
      Mask=1
      Masked Vector=&Multiply_VectorByFloat(
       Vector=&RandomRangeVector(Minimum=<300 15 200>, Maximum=<-300 -500 -1000>)
       Float=&FloatFromCurve(FloatCurve=DataInterface() {propertyValues {Curve {
           defaultValue=1
           keys [
            {arriveTangent=-3.5798, interpMode=RCIM_Cubic, leaveTangent=-3.5798, tangentMode=RCTM_User, time=0, value=1}
            {interpMode=RCIM_Cubic, tangentMode=RCTM_User, time=1, value=0}
           ]
          }}})
      )
      Passthrough Vector=&MaskVector(
       Mask Source=0
       Mask Parameter=?
       Mask=2
       Masked Vector=&Multiply_VectorByFloat(
        Vector=&RandomRangeVector(Minimum=<350 15 300>, Maximum=<-350 -500 -1000>)
        Float=&FloatFromCurve(FloatCurve=DataInterface() {propertyValues {Curve {
            defaultValue=1
            keys [
             {arriveTangent=-3.5798, interpMode=RCIM_Cubic, leaveTangent=-3.5798, tangentMode=RCTM_User, time=0, value=1}
             {interpMode=RCIM_Cubic, tangentMode=RCTM_User, time=1, value=0}
            ]
           }}})
       )
       Passthrough Vector=<0 0 0>
      )
     )
    ))
   &Drag(
    Drag=&MaskFloat(
     Mask Source=0
     Mask Parameter=?
     Mask=0
     Masked Float=&RandomRangeFloat(Minimum=8, Maximum=10)
     Passthrough Float=&MaskFloat(
      Mask Source=0
      Mask Parameter=?
      Mask=1
      Masked Float=&RandomRangeFloat(Minimum=8, Maximum=10)
      Passthrough Float=&MaskFloat(
       Mask Source=0
       Mask Parameter=?
       Mask=2
       Masked Float=&RandomRangeFloat(Minimum=8, Maximum=10)
       Passthrough Float=0
      )
     )
    )
    Use Linear Drag=true
    Use Rotational Drag=false
   )
   &CurlNoiseForce(
    Bias Noise=false
    Noise Frequency=&MaskFloat(
     Mask Source=0
     Mask Parameter=?
     Mask=0
     Masked Float=100
     Passthrough Float=&MaskFloat(
      Mask Source=0
      Mask Parameter=?
      Mask=1
      Masked Float=100
      Passthrough Float=&MaskFloat(Mask Source=0, Mask Parameter=?, Mask=2, Masked Float=100, Passthrough Float=0)
     )
    )
    Noise Strength=&MaskFloat(
     Mask Source=0
     Mask Parameter=?
     Mask=0
     Masked Float=400
     Passthrough Float=&MaskFloat(
      Mask Source=0
      Mask Parameter=?
      Mask=1
      Masked Float=400
      Passthrough Float=&MaskFloat(Mask Source=0, Mask Parameter=?, Mask=2, Masked Float=300, Passthrough Float=0)
     )
    )
    Pan Noise=false
   )
   &GravityForce(Gravity=&MaskVector(
     Mask Source=0
     Mask Parameter=?
     Mask=0
     Masked Vector=&MakeVector(
      X=0
      Y=0
      Z=&FloatFromCurve(FloatCurve=DataInterface() {propertyValues {Curve {keys [<at -0.0007 500>,<at 0.1 -350>,<at 0.5 -980>]}}}, Scale Curve=1.75)
     )
     Passthrough Vector=&MaskVector(
      Mask Source=0
      Mask Parameter=?
      Mask=1
      Masked Vector=&MakeVector(
       X=0
       Y=0
       Z=&FloatFromCurve(FloatCurve=DataInterface() {propertyValues {Curve {keys [<at -0.0007 500>,<at 0.1 -350>,<at 0.5 -980>]}}}, Scale Curve=1.75)
      )
      Passthrough Vector=&MaskVector(
       Mask Source=0
       Mask Parameter=?
       Mask=2
       Masked Vector=&MakeVector(
        X=0
        Y=0
        Z=&FloatFromCurve(
         FloatCurve=DataInterface() {propertyValues {Curve {keys [<at 0.0003 600>,<at 0.1 -350>,<at 0.5 -980>]}}}
         Scale Curve=&FloatFromCurve(FloatCurve=DataInterface() {propertyValues {Curve {keys [<at 0 0.5>,<at 1 1.5>]}}}, Scale Curve=3)
        )
       )
       Passthrough Vector=<0 0 0>
      )
     )
    ))
   &InheritVelocity(
    Inherited Velocity Amount Scale=&MaskVector(
     Mask Source=0
     Mask Parameter=?
     Mask=0
     Masked Vector=&MakeCustomVectorFromBool(
      Bool=&SetBoolByFloatComparison(A=&VectorLength(Vector=$User.OwnerLocalVelocity), B=301)
      True Vector=<0 0 0>
      False Vector=<0.01 -0.05 0.01>
     )
     Passthrough Vector=&MaskVector(
      Mask Source=0
      Mask Parameter=?
      Mask=1
      Masked Vector=&MakeCustomVectorFromBool(
       Bool=&SetBoolByFloatComparison(A=&VectorLength(Vector=$User.OwnerLocalVelocity), B=301)
       True Vector=<0 0 0>
       False Vector=<0.01 -0.05 0.01>
      )
      Passthrough Vector=&MaskVector(
       Mask Source=0
       Mask Parameter=?
       Mask=2
       Masked Vector=&MakeCustomVectorFromBool(
        Bool=&SetBoolByFloatComparison(A=&VectorLength(Vector=$User.OwnerLocalVelocity), B=301)
        True Vector=<0 0 0>
        False Vector=<0.01 -0.05 0.01>
       )
       Passthrough Vector=<0 0 0>
      )
     )
    )
    Speed Limit=true
    UseSpeedThreshold=false
   )
   &PointAttractionForce(
    Attraction Strength=&MaskFloat(
     Mask Source=0
     Mask Parameter=?
     Mask=0
     Masked Float=20
     Passthrough Float=&MaskFloat(
      Mask Source=0
      Mask Parameter=?
      Mask=1
      Masked Float=20
      Passthrough Float=&MaskFloat(Mask Source=0, Mask Parameter=?, Mask=2, Masked Float=20, Passthrough Float=0)
     )
    )
    Attractor Position Offset:Vector3f=<0 -300 -200>
    Kill Within Radius=false
    Module.Use Falloff=true
    UseAttractorOffset=true
   )
   &SolveForcesAndVelocity(Clamp Velocity=false, Limit Acceleration=false)
   &ScaleColor(
    Scale Alpha=&MergeHelper_FloatSwitch(
     0=&FloatFromCurve(FloatCurve=DataInterface() {propertyValues {Curve {
         defaultValue=1
         keys [
          {arriveTangent=-0.2109, interpMode=RCIM_Cubic, leaveTangent=-0.2109, tangentMode=RCTM_User, time=0.4765, value=1}
          {interpMode=RCIM_Cubic, tangentMode=RCTM_User, time=1, value=0}
         ]
        }}})
     1=&FloatFromCurve(FloatCurve=DataInterface() {propertyValues {Curve {
         defaultValue=1
         keys [
          {arriveTangent=-0.2109, interpMode=RCIM_Cubic, leaveTangent=-0.2109, tangentMode=RCTM_User, time=0.4765, value=1}
          {interpMode=RCIM_Cubic, tangentMode=RCTM_User, time=1, value=0}
         ]
        }}})
     2=&FloatFromCurve(
      FloatCurve=DataInterface() {propertyValues {Curve {
         defaultValue=1
         keys [
          <at 0 1 RCIM_Cubic>
          {arriveTangent=-1.3599, interpMode=RCIM_Cubic, leaveTangent=-1.3599, tangentMode=RCTM_User, time=0.3398, value=0.4425}
          <at 1 0 RCIM_Cubic>
         ]
        }}}
      Scale Curve=0.1
     )
     3=&FloatFromCurve(FloatCurve=DataInterface() {propertyValues {Curve {keys [<at 0 1>,<at 1 0>]}}})
     4=&FloatFromCurve(FloatCurve=DataInterface() {propertyValues {Curve {keys [<at 0 1 RCIM_Cubic>,<at 1 0 RCIM_Cubic>]}}})
     5=&FloatFromCurve(
      FloatCurve=DataInterface() {propertyValues {Curve {keys [
          {interpMode=RCIM_Cubic, leaveTangent=-1.9294, tangentMode=RCTM_Break, time=0, value=1}
          {interpMode=RCIM_Cubic, tangentMode=RCTM_Break, time=1, value=0}
         ]}}}
      Scale Curve=&RandomRangeFloat(Minimum=0.1, Maximum=0.3)
     )
     6=&FloatFromCurve(FloatCurve=DataInterface() {propertyValues {Curve {keys [
          {arriveTangent=51.5483, interpMode=RCIM_Cubic, leaveTangent=51.5483, tangentMode=RCTM_User, time=0, value=0}
          {arriveTangent=-0.1074, interpMode=RCIM_Cubic, leaveTangent=-0.1074, time=0.0647, value=1}
          {arriveTangentWeight=0.0616, interpMode=RCIM_Cubic, leaveTangentWeight=0.3948, tangentMode=RCTM_User, time=0.283, value=-0.0304}
          <at 1 0>
         ]}}})
     7=&FloatFromCurve(FloatCurve=DataInterface() {propertyValues {Curve {keys [<at 0 1>,<at 0.2892 0.8>,<at 1 0>]}}})
     Index=&SpawnGroup()
     Num=::NewEnumerator3
    )
    Scale RGB=&MergeHelper_VectorSwitch(
     0=&RandomRangeVector(Minimum=<0.8 0.8 0.8>, Maximum=<1.2 1.2 1.2>)
     1=&RandomRangeVector(Minimum=<0.8 0.8 0.8>, Maximum=<1.2 1.2 1.2>)
     2=&RandomRangeVector(Minimum=<0.8 0.8 0.8>, Maximum=<1.2 1.2 1.2>)
     3=<1 1 1>
     4=<1 1 1>
     Index=&SpawnGroup()
     Num=::NewEnumerator3
    )
    ScaleA=true
    ScaleRGB=true
   )
   /SetVariables(Particles.BalloonMesh DynamicParam:Vector4f=&MakeVector4(
     X=&FloatFromCurve(FloatCurve=DataInterface() {propertyValues {Curve {
         defaultValue=1
         keys [
          <at -0.0068 1 RCIM_Cubic>
          {arriveTangent=-1.5352, interpMode=RCIM_Cubic, leaveTangent=-1.5352, tangentMode=RCTM_User, time=0.4158, value=0.4226}
          {arriveTangent=-0.1366, interpMode=RCIM_Cubic, leaveTangent=-0.1366, tangentMode=RCTM_User, time=1, value=0}
         ]
        }}})
     Y=0
     Z=0
     W=0
    )) {isSetParameters=true}
   /SetVariables(Particles.BurstMesh DynamicParam:Vector4f=&MakeVector4(
     X=&FloatFromCurve(FloatCurve=DataInterface() {propertyValues {Curve {keys [{arriveTangent=0.0592, interpMode=RCIM_Cubic, time=0, value=0}, <at 1 1 RCIM_Cubic>]}}})
     Y=50
     Z=1
     W=1
    )) {isSetParameters=true}
   &BlendSpace("Position Blend (World <-> Local)":NiagaraFloat=&MergeHelper_FloatSwitch(
     0=&MakeCustomFloatFromBool(
      Bool=&SetBoolByFloatComparison(A=&VectorLength(Vector=$User.OwnerLocalVelocity), B=301)
      True Float=&FloatFromCurve(FloatCurve=DataInterface() {propertyValues {Curve {keys [<at 0.3 1>,<at 0.7 0.42>]}}})
      False Float=&FloatFromCurve(FloatCurve=DataInterface() {propertyValues {Curve {keys [<at 0.25 1>,<at 0.5 0>]}}})
     )
     1=&MakeCustomFloatFromBool(
      Bool=&SetBoolByFloatComparison(A=&VectorLength(Vector=$User.OwnerLocalVelocity), B=301)
      True Float=&FloatFromCurve(FloatCurve=DataInterface() {propertyValues {Curve {keys [<at 0.3 1>,<at 0.7 0.42>]}}})
      False Float=&FloatFromCurve(FloatCurve=DataInterface() {propertyValues {Curve {keys [<at 0.25 1>,<at 0.5 0>]}}})
     )
     2=&MakeCustomFloatFromBool(
      Bool=&SetBoolByFloatComparison(A=&VectorLength(Vector=$User.OwnerLocalVelocity), B=301)
      True Float=&FloatFromCurve(FloatCurve=DataInterface() {propertyValues {Curve {keys [<at 0.3 1>,<at 0.7 0.42>]}}})
      False Float=&FloatFromCurve(FloatCurve=DataInterface() {propertyValues {Curve {keys [<at 0.25 1>,<at 0.5 0>]}}})
     )
     3=&MakeCustomFloatFromBool(
      Bool=&SetBoolByFloatComparison(A=&VectorLength(Vector=$User.OwnerLocalVelocity), B=301)
      True Float=1
      False Float=&FloatFromCurve(FloatCurve=DataInterface() {propertyValues {Curve {keys [<at 0.25 1>,<at 0.5 0>]}}})
     )
     4=&MakeCustomFloatFromBool(
      Bool=&SetBoolByFloatComparison(A=&VectorLength(Vector=$User.OwnerLocalVelocity), B=301)
      True Float=1
      False Float=&FloatFromCurve(FloatCurve=DataInterface() {propertyValues {Curve {keys [<at 0.25 1>,<at 0.5 0>]}}})
     )
     5=&MakeCustomFloatFromBool(
      Bool=&SetBoolByFloatComparison(A=&VectorLength(Vector=$User.OwnerLocalVelocity), B=301)
      True Float=1
      False Float=&FloatFromCurve(FloatCurve=DataInterface() {propertyValues {Curve {keys [<at 0.25 1>,<at 0.5 0>]}}})
     )
     Index=&SpawnGroup()
     Num=::NewEnumerator3
     Translation Contribution:NiagaraFloat=1
     Rotation Contribution:NiagaraFloat=0
    ))
  ]
  rendererClasses [&NiagaraSpriteRendererProperties, &NiagaraMeshRendererProperties]
  renderers [
   {rendererClass=&NiagaraSpriteRendererProperties}
   {rendererClass=&NiagaraSpriteRendererProperties, rendererIndex=1}
   {rendererClass=&NiagaraMeshRendererProperties, rendererIndex=2}
   {rendererClass=&NiagaraMeshRendererProperties, rendererIndex=3}
   {rendererClass=&NiagaraSpriteRendererProperties, rendererIndex=4}
  ]
 }
 BrightBurst {
  emitterUpdateScript [
   &EmitterState(Life Cycle Mode=::NewEnumerator1, Loop Behavior=::NewEnumerator1, Loop Duration=1, UseLoopDelay=false)
   &SpawnBurst_Instantaneous(Use Spawn Probability=false, UseLoopCountLimit=false)
  ]
  particleSpawnScript [&InitializeParticle(
    Color=&FN_Select_Color_from_Time_Of_Day(
     Morning=<rgba 0.3125 0.3125 0.3125 1>
     Day=<rgba 0.4601 0.4903 0.5104 1>
     Evening=<rgba 0.3125 0.3125 0.3125 1>
     Night=<rgba 0.1648 0.2191 0.3229 1>
    )
    Lifetime=0.3
    Sprite Size Mode=::NewEnumerator3
    Uniform Sprite Size=75
    UsePositionOffset=false
    Write Material Random=true
   )]
  particleUpdateScript [
   &ParticleState()
   &ScaleColor(
    Scale Alpha=&FloatFromCurve(FloatCurve=DataInterface() {propertyValues {Curve {keys [
         {arriveTangent=51.5483, interpMode=RCIM_Cubic, leaveTangent=51.5483, tangentMode=RCTM_User, time=0, value=0}
         {arriveTangent=-0.1074, interpMode=RCIM_Cubic, leaveTangent=-0.1074, time=0.0647, value=1}
         {arriveTangentWeight=0.0616, interpMode=RCIM_Cubic, leaveTangentWeight=0.3948, tangentMode=RCTM_User, time=0.283, value=-0.0304}
         <at 1 0>
        ]}}})
    ScaleA=true
    ScaleRGB=true
   )
  ]
  rendererClasses [&NiagaraSpriteRendererProperties]
  renderers [{rendererClass=&NiagaraSpriteRendererProperties}]
 }
 DarkBurst {rendererClasses [], renderers []}
 SmokeLWE {rendererClasses [], renderers []}
}
summary {
 systemName=NS_Emote_RevoltCrush_BalloonPop
 emitters [
  {emitterName=AirStreaks, rendererClasses [&NiagaraSpriteRendererProperties]}
  {emitterName=DarkBurst, rendererClasses []}
  {emitterName=SmokeLWE, rendererClasses []}
  {
   emitterName=BalloonBitsCondensed
   rendererClasses [&NiagaraSpriteRendererProperties, &NiagaraMeshRendererProperties]
  }
  {emitterName=BrightBurst, rendererClasses [&NiagaraSpriteRendererProperties]}
 ]
 userVariables [
  User.bIsFrontEnd {type=NiagaraBool}
  User.bIsFrontEndPreview {type=NiagaraBool}
  User.OwnerLocalVelocity {type=Vector3f, description="", defaultValue {struct=Vector3f, value=<0 0 0>}}
 ]
}
