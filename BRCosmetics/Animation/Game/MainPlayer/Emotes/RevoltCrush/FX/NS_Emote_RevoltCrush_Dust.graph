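# Niagara system for the RevoltCrush traversal emote that kicks up dust and debris sprites from socket positions sampled through a socket reader.
# Both emitters switch spawn rate on the owner's local velocity, and dust drag eases as speed rises.
# Dust color follows time of day or a front-end preview palette.
# Distance-culled debris sprays outward under gravity, with collision gated by platform quality and shrinking late in life.
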NiagaraSystem /BRCosmetics/Animation/Game/MainPlayer/Emotes/RevoltCrush/FX/NS_Emote_RevoltCrush_Dust
dependencies {
 usedDataInterfaces [
  &NiagaraDataInterfaceCurve
  &NiagaraDataInterfaceColorCurve
  &NiagaraDataInterfacePlatformSet
  &NiagaraDataInterfaceSocketReader
 ]
 usedDynamicInputs [&MergeHelper_FloatSwitch, &SimulationPosition]
 usedModules [
  &SystemState
  &BlendSpaceHelper
  &self:SystemSpawnScript.SystemScriptSource.SystemScriptGraph.NiagaraNodeAssignment_0.SetVariables_#1
 ]
}
emitters {
 Default_Emote_TraversalDust {
  emitterSpawnScript [/SetVariables(Emitter.FE Particle Spawn Rate=70, Emitter.Max Traversal Speed=300, Emitter.Particle Spawn Rate=70) {isSetParameters=true}]
  emitterUpdateScript [
   &EmitterState()
   &SpawnRate(
    SpawnRate=&FN_FrontEndFloatModifier(
     Gameplay=&MakeCustomFloatFromBool(
      Bool=&SetBoolByFloatComparison(A=&VectorLength(Vector=$User.OwnerLocalVelocity), B=200)
      True Float=$Emitter.Particle Spawn Rate
      False Float=$Emitter.Idle Spawn Rate
     )
     Front End=$Emitter.FE Particle Spawn Rate
    )
    Use Spawn Probability=false
   )
  ]
  particleSpawnScript [
   &PartitionParticles()
   /SetVariables(Particles.VisibilityTag=&SpawnGroup()) {isSetParameters=true}
   &InitializeParticle(
    Color=&MakeCustomLinearColorFromBool(
     Bool=$User.bIsFrontEndPreview
     True Color=&MakeCustomLinearColorFromBool(
      Bool=&SetBoolByIntComparison(A=$Engine.QualityLevel, B=1, Comparison Type=::NewEnumerator7)
      True Color=&RandomRangeLinearColor(Minimum=<rgba 0.0625 0.0465 0.0397 0.4>, Maximum=<rgba 0.0938 0.0681 0.0571 0.4>)
      False Color=&RandomRangeLinearColor(Minimum=<rgba 0.7865 0.5204 0.3564 0.3>, Maximum=<rgba 0.776 0.5975 0.4406 0.4>)
     )
     False Color=&FNTimeOfDayColor(
      Night Color=&RandomRangeLinearColor(Minimum=<rgba 0.4167 0.2326 0.1584 0.2>, Maximum=<rgba 0.6979 0.4515 0.2981 0.4>)
      Morning Color=&RandomRangeLinearColor(Minimum=<rgba 0.349 0.2823 0.2399 0.2>, Maximum=<rgba 0.8125 0.6641 0.584 0.4>)
      Day Color=&RandomRangeLinearColor(Minimum=<rgba 0.2656 0.2242 0.1951 0.2>, Maximum=<rgba 0.4323 0.3628 0.3017 0.4>)
      Evening Color=&RandomRangeLinearColor(Minimum=<rgba 0.2917 0.2582 0.237 0.2>, Maximum=<rgba 0.6719 0.5913 0.5249 0.4>)
     )
    )
    Lifetime Max=0.6
    Lifetime Min=0.3
    Lifetime Mode=::NewEnumerator1
    Sprite Rotation Mode=::NewEnumerator2
    Sprite Size Mode=::NewEnumerator1
    Uniform Sprite Size Max=75
    Uniform Sprite Size Min=50
    UsePositionOffset=false
    Write Material Random=true
   ) {module=InitializeParticle001}
   &SocketLocation(
    Interpolate Socket Sampling=true
    SamplePosOffset=true
    Sampled Position Offset=<0 -8 -17>
    Sampled Position Offset Coordinate Space:ENiagaraSkelMeshTransforms=::NewEnumerator2
    Socket Index=$Particles.PartitionParticles.Partition
    Socket Reader=$System.RevoltCrush Socket Reader
    Socket Sampling Mode=::NewEnumerator3
    VelocitySpeedLimit=false
   )
   &AddVelocity(Velocity=<0 -2 0.2>, Velocity Speed Scale=&RandomRangeFloat(Minimum=300, Maximum=500)) {module=AddVelocity001}
   &CameraOffset(Camera Offset Amount=15)
  ]
  particleUpdateScript [
   &ParticleState()
   &Drag(
    Drag=&FloatFromCurve(
     FloatCurve=DataInterface() {propertyValues {Curve {
        defaultValue=1
        keys [{interpMode=RCIM_Cubic, tangentMode=RCTM_User, time=0, value=0.75}, {interpMode=RCIM_Cubic, tangentMode=RCTM_User, time=0.65, value=1}]
       }}}
     Scale Curve=&Lerp_Float(
      A=20
      B=8
      Alpha=&RemapRange(
       Input Value To Remap=&VectorLength(Vector=$User.OwnerLocalVelocity)
       Input Min=0
       Input Max=$Emitter.Max Traversal Speed
       Output Min=0
       Output Max=1
      )
     )
    )
    Use Linear Drag=true
    Use Rotational Drag=true
   )
   &ScaleSpriteSize(Uniform Curve Sprite Scale=DataInterface() {propertyValues {Curve {keys [
        {arriveTangent=0.2286, interpMode=RCIM_Cubic, leaveTangent=0.2286, tangentMode=RCTM_User, time=0, value=0.913}
        {arriveTangent=1.158, interpMode=RCIM_Cubic, leaveTangent=1.158, time=0.1696, value=1.0276}
        {arriveTangent=0.0238, interpMode=RCIM_Cubic, leaveTangent=0.0238, tangentMode=RCTM_User, time=0.3911, value=1.366}
        <at 0.7986 0.75 RCIM_Cubic>
       ]}}})
   &DynamicMaterialParameters(
    Index 0 Param 1=&Add_Float(
     A=&RandomRangeFloat(Minimum=0, Maximum=1)
     B=&FloatFromCurve(
      FloatCurve=DataInterface() {propertyValues {Curve {keys [
          <at 0 0 RCIM_Cubic>
          {arriveTangent=1.5947, interpMode=RCIM_Cubic, leaveTangent=1.5947, tangentMode=RCTM_User, time=0.2557, value=0.5257}
          <at 1 1 RCIM_Cubic>
         ]}}}
      CurveIndex=$Particles.Age
      Scale Curve=0.25
     )
    )
    Module.Index0Float0Write=true
    Module.Index0Float1Write=true
    Module.Index0Float2Write=true
    Module.Index0Float3Write=true
   ) {module=DynamicMaterialParameters001}
   &ScaleColor(
    Scale Alpha=&FloatFromCurve(
     FloatCurve=DataInterface() {propertyValues {Curve {keys [
         {interpMode=RCIM_Cubic, tangentMode=RCTM_User, time=0, value=0}
         {arriveTangentWeight=0.0754, interpMode=RCIM_Cubic, leaveTangentWeight=0.2297, tangentMode=RCTM_User, time=0.0584, value=1}
         {arriveTangent=-0.5014, interpMode=RCIM_Cubic, leaveTangent=-0.5014, tangentMode=RCTM_User, time=0.2914, value=1}
         {arriveTangent=0.0663, interpMode=RCIM_Cubic, leaveTangent=0.0663, tangentMode=RCTM_User, time=0.9, value=0}
        ]}}}
     Scale Curve=&FN_FrontEndFloatModifier(Gameplay=0.4, Front End=0.4)
    )
    ScaleA=true
    ScaleRGB=true
   )
   &BlendSpace(Rotation Contribution=0)
   &GravityForce(Gravity=&Multiply_VectorByFloat(
     Vector=<1 1 1>
     Float=&FloatFromCurve(
      FloatCurve=DataInterface() {propertyValues {Curve {defaultValue=1, keys [{tangentMode=RCTM_User, time=0, value=0}, {tangentMode=RCTM_User, time=1, value=1}]}}}
      Scale Curve=&RandomRangeFloat(Minimum=80, Maximum=170)
     )
    ))
   &SolveForcesAndVelocity(Clamp Velocity=false, Limit Acceleration=false)
  ]
  rendererClasses [&NiagaraSpriteRendererProperties]
  renderers [
   {rendererClass=&NiagaraSpriteRendererProperties}
   {rendererClass=&NiagaraSpriteRendererProperties, rendererIndex=1}
  ]
 }
 Emote_TraversalDebris {
  emitterSpawnScript [/SetVariables(Emitter.Max Traversal Speed=300, Emitter.Particle Spawn Rate=10) {isSetParameters=true}]
  emitterUpdateScript [
   &EmitterState(Enable Distance Culling=true, MaxDistance=1150, Module.bUseMaxDistance=true, Module.bUseMinDistance=false, Scalability Mode=::NewEnumerator1)
   &SpawnRate(
    Spawn Probability=0.7
    SpawnRate=&MakeCustomFloatFromBool(
     Bool=&SetBoolByFloatComparison(A=&VectorLength(Vector=$User.OwnerLocalVelocity), B=150)
     True Float=$Emitter.Particle Spawn Rate
     False Float=$Emitter.Idle Spawn Rate
    )
    Use Spawn Probability=true
   )
  ]
  particleSpawnScript [
   &PartitionParticles()
   &InitializeParticle(
    Color=&ColorFromCurve(
     ColorCurve=DataInterface() {propertyValues {
       AlphaCurve {keys [<at -0.0023 1 RCIM_Cubic>]}
       BlueCurve {keys [<at -0.0023 0.2122 RCIM_Cubic>,<at 0.1383 0.2122>,<at 0.2178 0.1859>,<at 0.4486 0.75>,<at 0.7126 0.1859>,<at 1.0234 0.8802>]}
       GreenCurve {keys [<at -0.0023 0.2325 RCIM_Cubic>,<at 0.1383 0.2325>,<at 0.2178 0.2005>,<at 0.4486 0.8456>,<at 0.7126 0.2122>,<at 1.0234 0.9474>]}
       RedCurve {keys [<at -0.0023 0.25 RCIM_Cubic>,<at 0.1383 0.25>,<at 0.2178 0.2188>,<at 0.4486 1>,<at 0.7126 0.2188>,<at 1.0234 1>]}
      }}
     CurveIndex=&UniformRangedFloat(Minimum=0, Maximum=1)
    )
    Lifetime=&UniformRangedFloat(Minimum=0.25, Maximum=0.75)
    Sprite Rotation Mode=::NewEnumerator2
    Sprite Size Mode=::NewEnumerator3
    Uniform Sprite Size=&MakeFloatFromVector2d(Vector 2D=&Vector2DFromFloat(Value=&UniformRangedFloat(Minimum=4, Maximum=10)))
    UsePositionOffset=false
    Write Material Random=true
   ) {module=InitializeParticle001}
   &SocketLocation(
    SamplePosOffset=true
    Sampled Position Offset=<0 20 0>
    Sampled Position Offset Coordinate Space:ENiagaraSkelMeshTransforms=::NewEnumerator2
    Socket Index=$Particles.PartitionParticles.Partition
    Socket Reader=$System.RevoltCrush Socket Reader
    Socket Sampling Mode=::NewEnumerator3
    VelocitySpeedLimit=false
   )
   &ShapeLocation(MaskSpawnGroup=false, Override Seed=false, Shape Origin=$Output.SocketLocation.SampledPosition, Sphere Radius=2, UseOwnerScale=true)
   &AddVelocity(
    Module.Use Falloff Distance=false
    Origin Offset=<0 3 0>
    Override Seed=false
    Velocity Mode=::NewEnumerator1
    Velocity Origin=$Output.SocketLocation.SampledPosition
    Velocity Speed=&RandomRangeFloat(Minimum=100, Maximum=200)
   )
  ]
  particleUpdateScript [
   &ParticleState()
   &Drag(Use Linear Drag=true, Use Rotational Drag=false)
   &GravityForce(Gravity=<0 0 -500>)
   &Collision(
    Age Colliding Particles=false
    Collision Enabled=&IsPlatformSetActive(New Platform Set=DataInterface() {propertyValues {Platforms {qualityLevelMask=-8}}})
    KillOnCollision=false
    Module.Control Rest State Via Surface Normal=false
    Module.EnableMaxCollisionCount=false
    Module.Randomize Collision Normal Vector=false
    Radius Calculation Type=::NewEnumerator1
   )
   &Collision(
    Age Colliding Particles=false
    Analytical Collision Plane Position 1:Vector3f=<0 0 -20>
    Analytical Collision Plane Space:ENiagaraCoordinateSpace=::Local
    CPU Collision Type=::NewEnumerator3
    Collision Enabled=&IsPlatformSetActive(New Platform Set=DataInterface() {propertyValues {Platforms {qualityLevelMask=-28}}})
    KillOnCollision=false
    Module.Control Rest State Via Surface Normal=false
    Module.EnableMaxCollisionCount=false
    Module.Randomize Collision Normal Vector=false
    Perform 2 Analytical Collision Queries=false
    Restitution=0
   ) {module=Collision001}
   &ScaleColor(Scale RGB=<1.2 1.2 1.2>, ScaleA=true, ScaleRGB=true)
   &ScaleSpriteSize(Scale Factor=&Vector2DFromFloat(Value=&FloatFromCurve(FloatCurve=DataInterface() {propertyValues {Curve {keys [
          <at 0.7 0.9701>
          {arriveTangent=-3.2149, interpMode=RCIM_Cubic, leaveTangent=-3.2149, tangentMode=RCTM_User, time=0.8778, value=0.9191}
          <at 1.0011 0.002>
         ]}}})))
   &SpriteRotationRate()
   &BlendSpace(Rotation Contribution=0)
   &SolveForcesAndVelocity(Clamp Velocity=false, Limit Acceleration=false)
   &ScaleSpriteSizeBySpeed(
    Max Scale Factor=&Vector2DFromFloat(Value=1)
    Min Scale Factor=&Vector2DFromFloat(Value=0.25)
    Velocity Threshold=200
   )
  ]
  rendererClasses [&NiagaraSpriteRendererProperties]
  renderers [{rendererClass=&NiagaraSpriteRendererProperties}]
 }
}
summary {
 systemName=NS_Emote_RevoltCrush_Dust
 emitters [
  {emitterName=Default_Emote_TraversalDust, rendererClasses [&NiagaraSpriteRendererProperties]}
  {emitterName=Emote_TraversalDebris, rendererClasses [&NiagaraSpriteRendererProperties]}
 ]
 userVariables [
  User.bIsFrontEnd {type=NiagaraBool}
  User.bIsFrontEndPreview {type=NiagaraBool}
  User.OwnerLocalVelocity {type=Vector3f, description="", defaultValue {struct=Vector3f, value=<0 0 0>}}
 ]
}
